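# Physics asset for the CyberFu Brigade backpack cosmetic, defining capsule bodies and angular limits for its nunchucks, chain and four "money" pieces.
# Each money piece and both nunchuck handles are constrained to the pack body with narrow, per-piece swing and twist limits.
# Three chain links join the upper handle to the lower one, forming a closed constraint loop through the pack rather than a free-hanging chain.
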PhysicsAsset /BRCosmetics/Accessories/FORT_Backpacks/Backpack_F_MED_CyberFu_Brigade/Meshes/F_MED_CyberFu_Brigade_Pack_FE_Physics
bodies {
 dyn_chain_01 [{
   shapeType=Capsule
   length=1.4265
   radius=0.9205
   shapeName=dyn_chain_01_capsule
   center=<1.0062 0.0477 0.0305>
   rotation=<rot 0 92.465 -88.6562>
  }]
 dyn_chain_02 [{
   shapeType=Capsule
   length=1.4388
   radius=0.9484
   shapeName=dyn_chain_02_capsule
   center=<1.1035 -0.0099 0.0191>
   rotation=<rot 0 -93.0353 -88.8218>
  }]
 dyn_chain_03 [{
   shapeType=Capsule
   length=1.4353
   radius=0.9521
   shapeName=dyn_chain_03_capsule
   center=<1.1035 0.0016 -0.0179>
   rotation=<rot 0 91.2593 -89.8413>
  }]
 dyn_money_back_l [{shapeType=Capsule, length=2, radius=2, shapeName=dyn_money_back_l_capsule}]
 dyn_money_back_r [{shapeType=Capsule, length=2.841, radius=1.3218, shapeName=dyn_money_back_r_capsule, rotation=<rot -8.7225 -90 5.923>}]
 dyn_money_front_l [{shapeType=Capsule, length=2, radius=2, shapeName=dyn_money_front_l_capsule}]
 dyn_money_front_r [{shapeType=Capsule, length=5.4424, radius=2.1773, shapeName=dyn_money_front_r_capsule, rotation=<rot 17.3083 -90.0955 -13.7774>}]
 dyn_nunchucks_down [{
   shapeType=Capsule
   length=22
   radius=1.5
   shapeName=dyn_nunchucks_down_capsule
   center=<0.0004 -0.0397 0>
   rotation=<rot 0 -89.4886 -86.4619>
  }]
 dyn_nunchucks_up [{
   shapeType=Capsule
   length=21
   radius=1.85
   shapeName=dyn_nunchucks_up_capsule
   center=<-0.8198 -0.1082 0.6142>
   rotation=<rot 0 85.9904 -87.4515>
  }]
 pack [{shapeType=Capsule, length=0, radius=1, shapeName=pack_capsule}]
}
constraints [
 {bone1Name=dyn_chain_03, bone2Name=dyn_nunchucks_down, swing1LimitDegrees=10, swing2LimitDegrees=10, twistLimitDegrees=10}
 {bone1Name=dyn_chain_02, bone2Name=dyn_chain_01, swing1LimitDegrees=10, swing2LimitDegrees=10, twistLimitDegrees=10}
 {bone1Name=dyn_money_front_r, bone2Name=pack, swing1LimitDegrees=20, swing2LimitDegrees=10, twistLimitDegrees=4}
 {bone1Name=dyn_money_front_l, bone2Name=pack, swing1LimitDegrees=20, swing2LimitDegrees=10, twistLimitDegrees=3}
 {bone1Name=dyn_money_back_r, bone2Name=pack, swing1LimitDegrees=10, swing2LimitDegrees=3, twistLimitDegrees=9}
 {bone1Name=dyn_money_back_l, bone2Name=pack, swing1LimitDegrees=15, swing2LimitDegrees=5, twistLimitDegrees=10}
 {bone1Name=dyn_chain_01, bone2Name=dyn_nunchucks_up, swing1LimitDegrees=10, swing2LimitDegrees=10, twistLimitDegrees=10}
 {bone1Name=dyn_nunchucks_up, bone2Name=pack, swing1LimitDegrees=5, swing2LimitDegrees=4, twistLimitDegrees=2}
 {bone1Name=dyn_nunchucks_down, bone2Name=pack, swing1LimitDegrees=13, swing2LimitDegrees=4, twistLimitDegrees=2}
 {bone1Name=dyn_chain_03, bone2Name=dyn_chain_02, swing1LimitDegrees=10, swing2LimitDegrees=10, twistLimitDegrees=10}
]
